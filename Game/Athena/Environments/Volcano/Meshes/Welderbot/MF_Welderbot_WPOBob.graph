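# Material function for the volcano environment's Welderbot that outputs an animated bobbing vector, named for use as World Position Offset.
# Three time-driven sines with differing periods form an XYZ wobble, weighted toward vertical and scaled by BobAmount.
# A saturated world-height gradient, starting Offset units below the actor's position and spanning VerticalMaskGradientSize, masks the motion so it fades in going upward.
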MaterialFunction /Game/Athena/Environments/Volcano/Meshes/Welderbot/MF_Welderbot_WPOBob
graph {nodes [
  n1 = WorldPosition()
  n2 = Subtract(A=ActorPositionWS(), B=ScalarParameter() {defaultValue=600, parameterName=Offset}) {constB=600}
  n3 = Divide(A=ComponentMask(None=Subtract(A=@n1.XYZ, B=@n2)) {b=true}, B=ScalarParameter() {defaultValue=1000, parameterName=VerticalMaskGradientSize}) {constB=300}
  n4 = Time()
  n5 = Sine(None=@n4) {period=1.92}
  n6 = Sine(None=@n4) {period=2.2}
  n7 = Multiply(A=AppendVector(A=AppendVector(A=@n5, B=@n6), B=Sine(None=@n4) {period=3.3}), B=Constant3Vector() {constant=<rgba 0.25 0.25 1 0>})
  n8 = ScalarParameter() {defaultValue=10, parameterName=BobAmount}
  FunctionOutput(None=Multiply(A=Saturate(None=@n3), B=Multiply(A=@n7, B=@n8)))
 ]}
